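; Base level-prefab Blueprint for Spire boss slag minions that destroys itself when one of its tracked spawned actors dies.
; On BeginPlay it calls the parent, then registers a callback for when all its actors have spawned.
; The callback binds each actor's OnDied, casting to the slag shockwave prop first and to BGA_Spire_Minor as fallback.
; The shared death handler destroys the prefab on any such death, without waiting for the rest.
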(event EventBeginPlay
  (|CallParentFunction)
  (CallFunction|CallwhenAllActorHaveSpawned (delegate Custom|OnAllActorSpawned)))

(event Custom|OnAllActorSpawned
  (params (SpawnedActors :type Array%20of%20Actor%20Object%20References))
  (bind _array_element (Utilities|Array|ForEachLoopwithBreak SpawnedActors)
    (:LoopBody
      (bind _asb_prop_spire_slag_shockwave (Utilities|Casting|CastToB_Prop_SpireSlag_Shockwave _array_element)
        (:then
          (EventDispatchers|AssignOnDied _asb_prop_spire_slag_shockwave (delegate Custom|OnDied_Event)))
        (:CastFailed
          (bind _asbga_spire_minor (Utilities|Casting|CastToBGA_Spire_Minor _array_element))
          (EventDispatchers|AssignOnDied _asbga_spire_minor (delegate Custom|OnDied_Event)))))))

(event Collision|EventActorBeginOverlap
  (params (OtherActor :type Actor%20Object%20Reference)))

(event EventTick
  (params (DeltaSeconds :type Float%20%28single-precision%29)))

(event Custom|OnDied_Event
  (params (DamagedActor :type Actor%20Object%20Reference) (Damage :type Float%20%28single-precision%29) (InstigatedBy :type Controller%20Object%20Reference) (DamageCauser :type Actor%20Object%20Reference) (HitLocation :type Vector) (FHitComponent :type Primitive%20Component%20Object%20Reference) (BoneName :type Name) (Momentum :type Vector))
  (Actor|DestroyActor))

(fn ConstructionScript
  (|CallParentFunction))
